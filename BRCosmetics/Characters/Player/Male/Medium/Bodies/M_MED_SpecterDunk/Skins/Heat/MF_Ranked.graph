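# Material function that composites a ranked badge with crest and gem onto the SpecterDunk Heat skin's incoming material attributes.
# Badge mask and normal textures drive normal blending, thin film, gem color, roughness, metallic, shadows and AO.
# A timed, eased shine sweep adds emissive.
# A static switch adds panning caustic highlights, dimming, time-of-day gem glow and opacity changes; outputs are the attributes and combined mask, while a diamond branch sits unconnected.
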MaterialFunction /BRCosmetics/Characters/Player/Male/Medium/Bodies/M_MED_SpecterDunk/Skins/Heat/MF_Ranked
graph {nodes [
  Constant()
  n1 = GetMaterialAttributes(None=FunctionInput() {bUsePreviewValueAsDefault=true, inputName=InMA, inputType=FunctionInput_MaterialAttributes}) {attributeGetTypes [#1]}
  n2 = Constant()
  n3 = NamedRerouteUsage()
  n4 = SetMaterialAttributes!(
   MaterialAttributes=@n1.MaterialAttributes
   Normal=MaterialFunctionCall(Normal=@n1.Normal, Flatness=LinearInterpolate(A=@n2, Alpha=@n3)) {materialFunction=&FlattenNormal}.Result
  ) {attributeSetTypes [#1]}
  n5 = NamedRerouteUsage()
  n6 = GetMaterialAttributes(None=MaterialFunctionCall!(
    Material=MaterialFunctionCall!(Material Attributes=@n4, Alpha=@n5) {materialFunction=/Game/Global/Materials/Cosmetics/Ranked/MF_RankedThinFilm}.Result
    Additional Normal=NamedRerouteUsage()
    Mask=NamedRerouteUsage()
   ) {materialFunction=/Engine/Functions/MaterialLayerFunctions/MatLayerBlend_BlendAngleCorrectedNormals}.Blended Material) {attributeGetTypes [#2, #3, #4]}
  n7 = LinearInterpolate(A=@n6.BaseColor, B=NamedRerouteUsage(), Alpha=NamedRerouteUsage())
  n8 = NamedRerouteUsage()
  n9 = NamedRerouteUsage()
  n10 = ScalarParameter() {parameterName=Shadow Toggle}
  n11 = NamedRerouteUsage()
  n12 = SetMaterialAttributes!(
   MaterialAttributes=@n6.MaterialAttributes
   BaseColor=LinearInterpolate(A=@n7, B=Multiply(A=@n7, B=@n8), Alpha=Multiply(A=@n9, B=@n10) {constB=0})
   Roughness=LinearInterpolate(
    A=@n6.Roughness
    B=ScalarParameter() {defaultValue=0.2, group=Ranked Badge, parameterName="Gem -  Roughness", sortPriority=15}
    Alpha=NamedRerouteUsage()
   ) {constB=0}
   Metallic=LinearInterpolate(
    A=LinearInterpolate(A=@n6.Metallic, Alpha=@n11) {constB=0.8}
    B=ScalarParameter() {defaultValue=0.8, group=Ranked Badge, parameterName="Gem -  Metallic", sortPriority=16}
    Alpha=NamedRerouteUsage()
   )
  ) {attributeSetTypes [#2, #3, #4]}
  n13 = GetMaterialAttributes(None=@n12) {attributeGetTypes [#5, #2]}
  n14 = NamedRerouteUsage()
  n15 = NamedRerouteUsage()
  n16 = VectorParameter() {group=Ranked Badge, parameterName=Gem - Backlit Color, sortPriority=5, defaultValue=<rgba 0.9 0.5197 0.0043 1>}
  n17 = NamedRerouteUsage()
  n18 = StaticSwitchParameter(
   True=@n13.EmissiveColor
   False=StaticSwitchParameter(True=@n13.EmissiveColor, False=LinearInterpolate(A=@n13.EmissiveColor, B=@n16.RGB, Alpha=@n17)) {defaultValue=true, group=Ranked Badge, parameterName=Color Based Backlight / Intensity Based, sortPriority=7}
  ) {defaultValue=true, group=Ranked Badge, parameterName=Use OG Emissive?}
  n19 = Multiply(
   A=Multiply(A=Multiply(A=@n14, B=@n15), B=@n18)
   B=ScalarParameter() {defaultValue=5, group=Ranked Badge, parameterName=Shining Efffect - Strength, sortPriority=14}
  )
  n20 = NamedRerouteUsage()
  n21 = Saturate(None=Multiply(A=OneMinus(None=@n20), B=NamedRerouteUsage()))
  n22 = ChannelMaskParameter(None=MaterialFunctionCall(X=NamedRerouteUsage(), Y=NamedRerouteUsage(), Z=NamedRerouteUsage()) {materialFunction=&MakeFloat3}.Result) {group=Ranked Badge, parameterName=Scaled Gem Map / Radial Shape / SDF Box, sortPriority=-1}
  n23 = LinearInterpolate(
   A=@n13.BaseColor
   B=Multiply(A=@n13.BaseColor, B=ScalarParameter() {defaultValue=0.2, parameterName=Gem Shadow - Strength})
   Alpha=MaterialFunctionCall(In=Multiply(A=@n21, B=@n22.R), Contrast=ScalarParameter() {parameterName="Contrast (S)"}) {materialFunction=&CheapContrast}.Result
  )
  n24 = Multiply(A=@n23, B=ScalarParameter() {defaultValue=0.5, group=Ranked Badge Adjustments, parameterName=Sunny Dim})
  n25 = Multiply(
   A=MaterialFunctionCall() {materialFunction=/Game/Weapons/WeaponSkins/MaterialFunction/FortniteSunnySideFresnel}.Result
   B=NamedRerouteUsage()
  )
  n26 = GetMaterialAttributes(None=SetMaterialAttributes!(MaterialAttributes=@n13.MaterialAttributes, EmissiveColor=Add(A=@n19, B=@n18), BaseColor=LinearInterpolate(A=@n23, B=@n24, Alpha=@n25)) {attributeSetTypes [#5, #2]}) {attributeGetTypes [#2]}
  n27 = Multiply(A=@n26.BaseColor, B=Saturate(None=NamedRerouteUsage()))
  n28 = NamedRerouteUsage()
  n29 = NamedRerouteUsage()
  n30 = Multiply(A=@n26.BaseColor, B=Max(A=Saturate(None=@n28), B=OneMinus(None=@n29)))
  n31 = ScalarParameter() {group=Ranked Badge, parameterName=AO Only on Crest}
  n32 = SetMaterialAttributes!(
   MaterialAttributes=@n26.MaterialAttributes
   BaseColor=LinearInterpolate(
    A=@n26.BaseColor
    B=LinearInterpolate(A=@n27, B=@n30, Alpha=@n31)
    Alpha=ScalarParameter() {defaultValue=1, group=Ranked Badge, parameterName=AO Strength}
   )
  ) {attributeSetTypes [#2]}
  n33 = GetMaterialAttributes(None=@n32) {attributeGetTypes [#2, #5]}
  n34 = LinearInterpolate(
   A=@n33.BaseColor
   B=Multiply(A=@n33.BaseColor, B=ScalarParameter() {defaultValue=1, group=Ranked Badge Adjustments, parameterName=CrestColorBrightness})
   Alpha=NamedRerouteUsage()
  )
  n35 = NamedRerouteUsage()
  n36 = AppendVector(
   A=ScalarParameter() {defaultValue=0.02, group=Ranked Badge Adjustments, parameterName=Highlight_Scale_U, sliderMax=1}
   B=ScalarParameter() {defaultValue=0.02, group=Ranked Badge Adjustments, parameterName=Highlight_Scale_V, sliderMax=1}
  )
  n37 = ScalarParameter() {group=Ranked Badge Adjustments, parameterName=Highlight_Pan_U, sliderMax=1, sliderMin=-1}
  n38 = ScalarParameter() {defaultValue=0.01, group=Ranked Badge Adjustments, parameterName=Highlight_Pan_V, sliderMax=1, sliderMin=-1}
  n39 = TextureSampleParameter2D!(UVs=Panner(Coordinate=Multiply(A=@n35, B=@n36), Time=Time(), Speed=AppendVector(A=@n37, B=@n38)), Bias=Constant() {r=-2}, Apply View MipBias) {group=Ranked Badge Adjustments, mipValueMode=TMVM_MipBias, parameterName=HighlightTexture, texture=/FortWater/Textures/T_Caustics_3layer}
  n40 = Power(Base=@n39.RGB) {constExponent=0.5}
  n41 = ScalarParameter() {defaultValue=1, group=Ranked Badge Adjustments, parameterName=Highlight_Scale, sliderMax=1}
  n42 = Add(A=@n34, B=Multiply(A=@n34, B=Multiply(A=@n40, B=@n41)))
  n43 = Multiply(A=@n42, B=ScalarParameter() {defaultValue=0.5, group=Ranked Badge Adjustments, parameterName=DimCrestBaseColor})
  n44 = Add(A=LinearInterpolate(A=@n33.EmissiveColor, B=@n33.EmissiveColor, Alpha=NamedRerouteUsage()), B=Multiply(A=@n43, B=NamedRerouteUsage()))
  n45 = Multiply(
   A=ScalarParameter() {defaultValue=0.5, group=Ranked Badge Adjustments, parameterName=GemDimEmissive}
   B=MaterialFunctionCall(
    Value_Morning=ScalarParameter() {defaultValue=1.5, group=Ranked Badge Adjustments, parameterName=Gem TOD - Morning}
    Value_Day=ScalarParameter() {defaultValue=1.25, group=Ranked Badge Adjustments, parameterName=Gem TOD - Day}
    Value_Evening=ScalarParameter() {defaultValue=1.5, group=Ranked Badge Adjustments, parameterName=Gem TOD - Evening}
    Value_Night=ScalarParameter() {defaultValue=2, group=Ranked Badge Adjustments, parameterName=Gem TOD - Night}
   ) {materialFunction=&Scalar_Time_of_Day}.Result
  )
  n46 = Power(Base=NamedRerouteUsage()) {constExponent=5}
  n47 = LinearInterpolate(A=@n42, B=@n43, Alpha=NamedRerouteUsage())
  n48 = ScalarParameter() {defaultValue=0.75, group=Ranked Badge Adjustments, parameterName=GemDimBaseColor}
  n49 = SetMaterialAttributes!(
   MaterialAttributes=@n33.MaterialAttributes
   OpacityMask
   BaseColor=LinearInterpolate(A=@n47, B=Multiply(A=@n47, B=@n48), Alpha=NamedRerouteUsage())
   EmissiveColor=LinearInterpolate(A=@n44, B=Multiply(A=@n44, B=@n45), Alpha=NamedRerouteUsage())
   Opacity=LinearInterpolate(Alpha=@n46) {constA=1, constB=0}
  ) {attributeSetTypes [
    #6
    #2
    #5
    #7
   ]}
  FunctionOutput(None=StaticSwitchParameter(True=@n49, False=@n32) {group=Ranked Badge Adjustments, parameterName=UseAdditionalAdjustments}) {outputName=MA}
  FunctionOutput(None=@n46) {outputName=CombindMask, sortPriority=2}
  n50 = Constant()
  If(A=@n46, "A > B"=Constant() {r=1}, "A == B"=@n50, "A < B"=@n50) {constB=0.1}
  LinearInterpolate(A=@n6.BaseColor, Alpha=@n11) {constB=0}
  n51 = NamedRerouteUsage()
  n52 = MaterialFunctionCall(UVs=@n51, Texture Scale=ScalarParameter() {defaultValue=0.624, group=Ranked Badge, parameterName=Gem Shape - Box Scale, sortPriority=0}) {materialFunction=&ScaleUVsByCenter}
  n53 = Constant() {r=-0.125}
  n54 = Sine(None=Multiply(A=MaterialFunctionCall(UVs=@n52.UVs, "Rotation Angle (0-1)"=@n53) {materialFunction=&CustomRotator}.Rotated Values) {constB=0.5})
  n55 = Min(A=ComponentMask(None=@n54) {r=true}, B=ComponentMask(None=@n54) {g=true})
  NamedRerouteDeclaration(None=OneMinus(None=Step(Y=@n55) {constX=0.1})) {name=Box, nodeColor=<rgba 0.3594 1 0 1>}
  n56 = Power(Base=NamedRerouteUsage(), Exp=ScalarParameter() {defaultValue=1.5, group=Ranked Badge, parameterName=Gem Color - Transition, sortPriority=6})
  n57 = ScalarParameter() {defaultValue=1, group=Ranked Badge, parameterName=Gem Color - Strength, sortPriority=7}
  NamedRerouteDeclaration(None=Multiply(A=NamedRerouteUsage(), B=Multiply(A=@n56, B=@n57))) {name=Gem Color Mask, nodeColor=<rgba 0 1 0.2734 1>}
  n58 = Power(Base=NamedRerouteUsage(), Exp=ScalarParameter() {defaultValue=8, group=Ranked Badge, parameterName=Gem Brightness - Power, sortPriority=8}) {constExponent=6}
  NamedRerouteDeclaration(None=Multiply(A=Saturate(None=@n58), B=ScalarParameter() {defaultValue=1, group=Ranked Badge, parameterName=Gem Brightness - Strength, sortPriority=9}) {constB=0.75}) {name=Gem Brightness Mask, nodeColor=<rgba 1 0 0.0469 1>}
  n59 = Multiply(A=MaterialFunctionCall(UVs=NamedRerouteUsage(), "Rotation Angle (0-1)"=Constant() {r=-0.125}) {materialFunction=&CustomRotator}.Rotated Values) {constB=0.5}
  n60 = MaterialFunctionCall("Value (0-1)"=MaterialFunctionCall!(Animation Duration=Constant() {r=2}, Delay=Constant() {r=2}, Reset to Default) {materialFunction=/FortUILibrary/MaterialFunctions/Time/MF_UI_AnimationDelay}.Result) {materialFunction=&MF_UI_EaseCurves}
  n61 = Frac(None=Subtract(A=Add(A=@n59), B=Subtract(A=@n60.EaseOutCubic) {constB=0.5}))
  n62 = Constant() {r=0.1}
  MaterialExpressionRemap_4 = MaterialXRemap!(None=Subtract(A=Sine(None=@n61), B=OneMinus(None=@n62)) {constB=0.5}, InputLow, InputHigh=@n62, TargetLow, TargetHigh)
  NamedRerouteDeclaration(None=ComponentMask(None=Saturate(None=@MaterialExpressionRemap_4)) {r=true}) {name=Shining Effect, nodeColor=<rgba 0 0 0 1>}
  n63 = NamedRerouteDeclaration(None=FunctionInput() {inputName=In_UV, inputType=FunctionInput_Vector2}) {name=UVs, nodeColor=<rgba 0.8984 1 0 1>}
  n64 = TextureSampleParameter2D!(UVs=@n63, Apply View MipBias) {
   group=Ranked Badge
   parameterName=Texture - Ranked Badge Mask
   samplerSource=SSM_Clamp_WorldGroupSettings
   samplerType=SAMPLERTYPE_LinearColor
   sortPriority=2
   texture=/BRCosmetics/Characters/Player/Female/Medium/Bodies/F_MED_RankedTrooper/Textures/RankedBadges/Mask/T_RankedTrooper_Badge_Diamond_Mask
  }
  NamedRerouteDeclaration(None=@n64.B) {name=Gem Brightness Map, nodeColor=<rgba 0.8516 0 1 1>}
  NamedRerouteDeclaration(None=@n64.G) {name=Gem Map, nodeColor=<rgba 0 1 0.5312 1>}
  NamedRerouteDeclaration(None=NamedRerouteDeclaration(None=Max(A=@n64.R, B=@n64.R)) {name=Combined Mask, nodeColor=<rgba 1 0 0.2812 1>}) {name=Scaled Gem Map, nodeColor=<rgba 0 1 0.5312 1>}
  NamedRerouteDeclaration(None=MaterialFunctionCall(
    UVs=@n51
    Radius=ScalarParameter() {defaultValue=0.29, group=Ranked Badge, parameterName=Gem Shape - Circle Radius, sortPriority=11}
    Density=ScalarParameter() {defaultValue=10, group=Ranked Badge, parameterName=Gem Shape - Circle Density, sortPriority=12}
   ) {materialFunction=&RadialGradientExponential}.RadialGradientExponential) {name=Circle, nodeColor=<rgba 0.6641 0 1 1>}
  n65 = VectorParameter() {group=Ranked Badge, parameterName=Gem Color - Main, sortPriority=3, defaultValue=<rgba 0.4397 0.9823 0.1845 1>}
  n66 = VectorParameter() {group=Ranked Badge, parameterName=Gem Color - Secondary, sortPriority=4, defaultValue=<rgba 0.74 0.3619 0.0616 1>}
  n67 = LinearInterpolate(A=LinearInterpolate(B=@n65.RGB, Alpha=NamedRerouteUsage()), B=@n66.RGB, Alpha=NamedRerouteUsage())
  n68 = VectorParameter() {group=Ranked Badge, parameterName=Gem - Backlit Color, sortPriority=5, defaultValue=<rgba 0.9 0.5197 0.0043 1>}
  n69 = NamedRerouteUsage()
  n70 = LinearInterpolate(A=@n67, B=@n68.RGB, Alpha=@n69)
  n71 = Add(A=LinearInterpolate(B=@n67, Alpha=@n69) {desc=Gem Highlight}, B=@n67)
  NamedRerouteDeclaration(None=Multiply(
    A=StaticSwitchParameter(True=@n70, False=@n71) {defaultValue=true, group=Ranked Badge, parameterName=Color Based Backlight / Intensity Based, sortPriority=7}
    B=ScalarParameter() {defaultValue=1, group=Ranked Badge, parameterName=Gem Color - Overall Strength, sortPriority=10}
   )) {name=Final Gem Color, nodeColor=<rgba 0 1 0.6484 1>}
  n72 = Multiply(
   A=NamedRerouteDeclaration(None=@n64.R) {name=Crest Mask, nodeColor=<rgba 0 0.4375 1 1>}
   B=NamedRerouteDeclaration(None=@n64.B) {name=Gem Color Map, nodeColor=<rgba 0.8281 1 0 1>}
  )
  n73 = ScalarParameter() {defaultValue=5, parameterName=Crest Shadow}
  NamedRerouteDeclaration(None=Saturate(None=Multiply(A=@n72, B=@n73))) {name=Crest Shadow, nodeColor=<rgba 0 1 0.6953 1>}
  n74 = TextureSampleParameter2D(UVs=@n63) {
   group=Ranked Badge
   parameterName=Texture - Ranked Badge NRM
   samplerSource=SSM_Clamp_WorldGroupSettings
   samplerType=SAMPLERTYPE_LinearColor
   sortPriority=1
   texture=/BRCosmetics/Characters/Player/Female/Medium/Bodies/F_MED_RankedTrooper/Textures/RankedBadges/T_RankedTrooper_Badge_Bronze_N
  }
  NamedRerouteDeclaration(None=@n74.B) {name=AO Mask, nodeColor=<rgba 0.875 1 0 1>}
  n75 = Multiply(A=AppendVector(A=@n74.R, B=@n74.R)) {constB=2}
  NamedRerouteDeclaration(None=DeriveNormalZ(InXY=Subtract(A=@n75))) {name=NRM, nodeColor=<rgba 1 0 0.1172 1>}
  Reroute(None=@n6.Roughness)
  Reroute(None=LinearInterpolate(A=@n6.Roughness, Alpha=@n11) {constB=0.1})
  n76 = GetMaterialAttributes(None=MaterialFunctionCall!(UV=NamedRerouteUsage(), Material=@n12, TechArtMask=@n20) {materialFunction=/BRCosmetics/Weapons/FORT_Melee/Pickaxe_RankedDefender/Materials/MF_NewDiamond}.Material) {attributeGetTypes [#5, #2]}
  n77 = ScalarParameter() {parameterName=Diamond - Emissive Contrast}
  SetMaterialAttributes!(
   MaterialAttributes=@n76.MaterialAttributes
   EmissiveColor=LinearInterpolate(
    A=@n76.EmissiveColor
    B=MaterialFunctionCall!(In=@n76.EmissiveColor, Contrast=@n77) {materialFunction=/Engine/Functions/Engine_MaterialFunctions01/ImageAdjustment/CheapContrast_RGB}.Result
    Alpha=NamedRerouteUsage()
   )
   BaseColor
  ) {attributeSetTypes [#5, #2]}
  TextureCoordinate()
 ]}
