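# Parent material for the Pulse skin of the SpecterDunk character body, wrapping skin-specific edits between the shared CharacterParent PreFX and PostFX functions.
# Texture masks drive accent-metal tinting, roughness, metallic and specular offsets, and body and arm emissives with Fresnel and pattern.
# Sphere masks on pre-skinned local position create color gradients that stay fixed under animation.
# MF_Ranked is layered on top, using UV-region tests to choose separately placed decal coordinates.
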Material /BRCosmetics/Characters/Player/Male/Medium/Bodies/M_MED_SpecterDunk/Skins/Pulse/M_SpecterDunk_Pulse_Parent
graph {
 properties {
  bAutomaticallySetUsageInEditor=false
  bUseMaterialAttributes=true
  bUsedWithMeshDeformer=true
  bUsedWithMorphTargets=true
  bUsedWithSkeletalMesh=true
  pixelDepthOffsetMode=PDOM_Legacy
  shadingModel=MSM_FromMaterialExpression
  usedShadingModels=Default Lit | Subsurface | Clear Coat | Subsurface Profile | Cloth
 }
 nodes [
  n1 = NamedRerouteUsage()
  n2 = SphereMask(
   A=VertexInterpolator(VS=LocalPosition() {localOrigin=InstancePreSkinning}.XYZ).PS
   B=Constant4Vector() {constant=<rgba -67.0995 13.9892 104.1414 1>}.RGB
   Radius=Constant() {r=51.75}
   Hardness=Constant() {r=1}
  )
  n3 = Constant() {r=2}
  n4 = LinearInterpolate(
   A=LinearInterpolate(
    A=VectorParameter() {parameterName=AccentMetalsColor2, defaultValue=<rgba 0.4 0.4 0.4 1>}.RGB
    B=VectorParameter() {parameterName=AccentMetalsColor, defaultValue=<rgba 1.5 1.5 1.5 1>}.RGB
    Alpha=Power(Base=@n2, Exp=@n3)
   )
   B=VectorParameter() {parameterName=AccentMetalsColor3, defaultValue=<rgba 0.32 0.32 0.32 1>}.RGB
   Alpha=NamedRerouteUsage()
  )
  n5 = LinearInterpolate(
   A=LinearInterpolate(A=@n1, B=Multiply(A=@n1, B=@n4), Alpha=NamedRerouteUsage())
   B=Multiply(A=@n1, B=Constant() {r=0.5})
   Alpha=NamedRerouteUsage()
  )
  n6 = Multiply(A=@n5, B=NamedRerouteUsage())
  n7 = Constant() {r=0.5}
  n8 = SetMaterialAttributes!(
   MaterialAttributes=NamedRerouteUsage()
   Base Color=LinearInterpolate(A=@n5, B=LinearInterpolate(A=@n5, B=@n6, Alpha=@n7), Alpha=NamedRerouteUsage())
   Emissive Color=NamedRerouteUsage()
   Metallic=NamedRerouteUsage()
   Specular=NamedRerouteUsage()
   Roughness=NamedRerouteUsage()
   Opacity
  ) {attributeSetTypes [
    #1
    #2
    #3
    #4
    #5
    #6
   ]}
  n9 = ComponentMask(None=TextureCoordinate()) {r=true}
  n10 = Constant() {r=0.9}
  n11 = Constant() {r=0.2}
  n12 = TextureCoordinate()
  n13 = NamedRerouteUsage()
  n14 = If(
   A=SphereMask(A=ComponentMask(None=@n12) {r=true}, B=Constant() {r=0.15}, Radius=Constant() {r=0.2})
   "A > B"=NamedRerouteUsage()
   "A == B"=@n13
   "A < B"=@n13
  ) {constB=0.1}
  n15 = MaterialFunctionCall!(
   In_UV=If(A=SphereMask(A=@n9, B=@n10, Radius=@n11), "A > B"=NamedRerouteUsage(), "A == B"=@n14, "A < B"=@n14) {constB=0.01}
   InMA=MakeMaterialAttributes!(
    BaseColor
    Metallic
    Specular
    Roughness
    Anisotropy
    EmissiveColor
    Opacity
    OpacityMask
    Normal
    Tangent
    WorldPositionOffset
    SubsurfaceColor
    ClearCoat
    ClearCoatRoughness
    AmbientOcclusion
    Refraction
    CustomizedUVs_0
    CustomizedUVs_1
    CustomizedUVs_2
    CustomizedUVs_3
    CustomizedUVs_4
    CustomizedUVs_5
    CustomizedUVs_6
    CustomizedUVs_7
    PixelDepthOffset
    ShadingModel
    Displacement
   )
  ) {materialFunction=/BRCosmetics/Characters/Player/Male/Medium/Bodies/M_MED_SpecterDunk/Skins/Heat/MF_Ranked}
  n16 = MaterialFunctionCall() {materialFunction=&MF_CharacterParent_2023_PreFX}
  MaterialFunctionCall(
   Material=MaterialFunctionCall!(Base Material=@n8, Top Material=@n15.MA, Alpha=@n15.MA) {materialFunction=/Engine/Functions/MaterialLayerFunctions/MatLayerBlend_Standard}.Blended Material
   M=@n16.M
   Pre Skinned Local Position=@n16.M
   Distance Blend=@n16.M
   Fresnel=@n16.M
  ) {materialFunction=&MF_CharacterParent_2023_PostFX}
  n17 = NamedRerouteUsage()
  n18 = Power(Base=NamedRerouteUsage(), Exp=Constant() {r=2.5})
  n19 = SphereMask(
   A=VertexInterpolator(VS=LocalPosition() {localOrigin=InstancePreSkinning}.XYZ).PS
   B=Constant4Vector() {constant=<rgba 53.2776 7.5449 90.954 1>}.RGB
   Radius=Constant() {r=50}
   Hardness=Constant() {r=0.7}
  )
  n20 = Multiply(A=@n17, B=LinearInterpolate(B=Saturate(None=@n18), Alpha=OneMinus(None=@n19)) {constA=1})
  n21 = Constant() {r=0.9}
  NamedRerouteDeclaration(None=LinearInterpolate(A=NamedRerouteUsage(), B=LinearInterpolate(A=@n17, B=@n20, Alpha=@n21), Alpha=NamedRerouteUsage())) {name=Out_Emissive, nodeColor=<rgba 1 0.3984 0 1>}
  n22 = GetMaterialAttributes(None=@n16.Result) {attributeGetTypes [
    #1
    #2
    #3
    #4
    #5
    #6
   ]}
  NamedRerouteDeclaration(None=@n22.Base Color) {name=IN_BaseColor, nodeColor=<rgba 0 1 0.2969 1>}
  n23 = TextureSample() {
   samplerType=SAMPLERTYPE_Masks
   texture=/BRCosmetics/Characters/Player/Male/Medium/Bodies/M_MED_SpecterDunk/Textures/T_SpecterDunk_Body_TAMask2
  }
  NamedRerouteDeclaration(None=@n23.R) {name=BodyEmissiveMask, nodeColor=<rgba 1 0.8438 0 1>}
  n24 = LinearInterpolate(
   A=VectorParameter() {parameterName=BodyEmissive_ColorA, defaultValue=<rgba 3 0.4096 0.9878 1>}.RGB
   B=VectorParameter() {parameterName=BodyEmissive_ColorB, defaultValue=<rgba 1 0 0.2582 1>}.RGB
   Alpha=NamedRerouteUsage()
  )
  n25 = VectorParameter() {parameterName=BodyEmissive_Fresnel, defaultValue=<rgba 1 0 0.264 1>}
  n26 = Fresnel(ExponentIn=ScalarParameter() {defaultValue=0.2892, parameterName=BodyEmissive_Power})
  NamedRerouteDeclaration(None=Multiply(A=LinearInterpolate(A=@n24, B=@n25.RGB, Alpha=@n26), B=NamedRerouteUsage())) {name=BodyEmissive, nodeColor=<rgba 1 0 0.6797 1>}
  n27 = LocalPosition() {localOrigin=InstancePreSkinning}
  n28 = Power(
   Base=SphereMask(
    A=VertexInterpolator(VS=@n27.XYZ).PS
    B=Constant4Vector() {constant=<rgba 53.2776 7.5449 90.954 1>}.RGB
    Radius=Constant() {r=51.75}
    Hardness=Constant()
   )
   Exp=ScalarParameter() {defaultValue=3, parameterName=GradientPower}
  )
  n29 = Multiply(A=@n28, B=NamedRerouteUsage())
  n30 = Saturate(None=Saturate(None=Add(A=@n28, B=@n29)))
  n31 = LinearInterpolate(
   A=VectorParameter() {parameterName=ArmEmissive_ColorA, defaultValue=<rgba 0.3462 0.0425 0.85 1>}.RGB
   B=VectorParameter() {parameterName=ArmEmissive_ColorB, defaultValue=<rgba 4 0.08 1.5612 1>}.RGB
   Alpha=@n30
  )
  n32 = LinearInterpolate(
   A=VectorParameter() {parameterName=ArmPattern_ColorB, defaultValue=<rgba 0.196 0.0532 0.3325 1>}.RGB
   B=VectorParameter() {parameterName=ArmPattern_Color, defaultValue=<rgba 2 0.0584 0.4431 1>}.RGB
   Alpha=@n30
  )
  n33 = TextureSample(UVs=MaterialFunctionCall() {materialFunction=/BRCosmetics/Characters/Player/Male/Medium/Bodies/M_MED_SpecterDunk/Materials/MF_AtlasUV_2x2}.AtlasUV_2x2) {
   samplerType=SAMPLERTYPE_Masks
   texture=/BRCosmetics/Characters/Player/Male/Medium/Bodies/M_MED_SpecterDunk/Textures/T_SpecterDunk_Body_PatternA
  }
  n34 = Constant() {r=2}
  NamedRerouteDeclaration(None=LinearInterpolate(
    A=LinearInterpolate(A=@n31, B=@n32, Alpha=@n33.R)
    B=VectorParameter() {parameterName=ArmEmissive_Fresnel, defaultValue=<rgba 0.1913 0.0201 0.171 1>}.RGB
    Alpha=Fresnel(ExponentIn=@n34)
   )) {name=ArmEmissive, nodeColor=<rgba 1 0.6797 0 1>}
  n35 = TextureSample() {
   samplerType=SAMPLERTYPE_Masks
   texture=/BRCosmetics/Characters/Player/Male/Medium/Bodies/M_MED_SpecterDunk/Textures/T_SpecterDunk_Body_TAMask
  }
  NamedRerouteDeclaration(None=@n35.B) {name=ArmMask, nodeColor=<rgba 0.5234 0 1 1>}
  NamedRerouteDeclaration(None=@n22.Metallic) {name=In_Metallic, nodeColor=<rgba 0.1953 0 1 1>}
  NamedRerouteDeclaration(None=@n22.Specular) {name=In_Specular, nodeColor=<rgba 0.125 1 0 1>}
  NamedRerouteDeclaration(None=@n22.Roughness) {name=In_Roughness, nodeColor=<rgba 1 0 0.4453 1>}
  n36 = Add(
   A=NamedRerouteUsage()
   B=LinearInterpolate(B=ScalarParameter() {defaultValue=-0.3, parameterName=RoughnessOffset}, Alpha=NamedRerouteUsage())
  )
  n37 = ScalarParameter() {defaultValue=0.2, parameterName=Sole_RoughnessMIn}
  n38 = ScalarParameter() {defaultValue=0.1, parameterName=Sole_RoughnessMax}
  n39 = NamedRerouteUsage()
  NamedRerouteDeclaration(None=LinearInterpolate(A=Saturate(None=@n36), B=LinearInterpolate(A=@n37, B=@n38, Alpha=@n39), Alpha=NamedRerouteUsage())) {name=OutRoughness, nodeColor=<rgba 0 0.7656 1 1>}
  n40 = NamedRerouteUsage()
  n41 = LinearInterpolate(
   A=LinearInterpolate(B=ScalarParameter() {defaultValue=0.5, parameterName=MetallicOffset}, Alpha=NamedRerouteUsage())
   B=Constant() {r=1}
   Alpha=NamedRerouteUsage()
  )
  NamedRerouteDeclaration(None=Saturate(None=Add(A=@n40, B=@n41))) {name=OutMetallic, nodeColor=<rgba 1 0.9141 0 1>}
  NamedRerouteDeclaration(None=@n22.MaterialAttributes) {name=In_MA, nodeColor=<rgba 1 0 0.0234 1>}
  n42 = NamedRerouteUsage()
  n43 = LinearInterpolate(B=ScalarParameter() {defaultValue=0.5, parameterName=SpecularOffset}, Alpha=NamedRerouteUsage())
  NamedRerouteDeclaration(None=Saturate(None=Add(A=@n42, B=@n43))) {name=OutSpecular, nodeColor=<rgba 1 0.9141 0 1>}
  NamedRerouteDeclaration(None=@n23.G) {name=Accent1, nodeColor=<rgba 0 1 0.2031 1>}
  n44 = TextureCoordinate()
  n45 = Constant() {r=0.815}
  n46 = TextureCoordinate()
  n47 = Multiply(
   A=SphereMask(A=ComponentMask(None=@n44) {r=true}, B=Constant() {r=0.65}, Radius=Constant() {r=0.08})
   B=SmoothStep(Min=@n45, Max=@n45, Value=ComponentMask(None=@n46) {g=true}) {constMax=0.7, constMin=0.7}
  )
  n48 = SphereMask(A=VertexInterpolator(VS=LocalPosition().Z).PS, B=Constant()) {attenuationRadius=25}
  NamedRerouteDeclaration(None=Multiply(A=Multiply(A=@n47, B=@n48), B=NamedRerouteUsage())) {name=Sole mask, nodeColor=<rgba 1 0.5391 0 1>}
  n49 = TextureCoordinate()
  n50 = AppendVector(
   A=ScalarParameter() {defaultValue=0.016, group=Ranked, parameterName=Back_UV_OffsetX}
   B=ScalarParameter() {defaultValue=0.34, group=Ranked, parameterName=Back_UV_OffsetY}
  )
  n51 = AppendVector(
   A=ScalarParameter() {defaultValue=0.21, group=Ranked, parameterName=Back_UV_ScaleX}
   B=ScalarParameter() {defaultValue=0.22, parameterName=Back_UV_ScaleY}
  )
  n52 = ScalarParameter() {defaultValue=0.62, group=Ranked, parameterName=UV_Global_Scale}
  NamedRerouteDeclaration(None=MaterialFunctionCall(UVs=Add(A=@n49, B=@n50), Texture Scale=Multiply(A=@n51, B=@n52)) {materialFunction=&ScaleUVsByCenter}.UVs) {name=UVBack, nodeColor=<rgba 1 0.2109 0 1>}
  n53 = TextureCoordinate()
  n54 = AppendVector(
   A=ScalarParameter() {defaultValue=0.335, group=Ranked, parameterName=Chest_UV_OffsetX}
   B=ScalarParameter() {defaultValue=0.34, group=Ranked, parameterName=Chest_UV_OffsetY}
  )
  n55 = AppendVector(
   A=ScalarParameter() {defaultValue=0.24, group=Ranked, parameterName=Chest_UV_ScaleX}
   B=ScalarParameter() {defaultValue=0.25, parameterName=Chest_UV_ScaleY}
  )
  NamedRerouteDeclaration(None=MaterialFunctionCall(UVs=Add(A=@n53, B=@n54), Texture Scale=Multiply(A=@n52, B=@n55)) {materialFunction=&ScaleUVsByCenter}.UVs) {name=UV_Chest, nodeColor=<rgba 0 0.1094 1 1>}
  n56 = NamedRerouteUsage()
  n57 = NamedRerouteUsage()
  n58 = NamedRerouteUsage()
  NamedRerouteDeclaration(None=LinearInterpolate(A=@n56, B=Constant(), Alpha=Max(A=@n57, B=@n58))) {name=Out_BaseColor, nodeColor=<rgba 0 0.3438 1 1>}
  n59 = TextureCoordinate()
  n60 = AppendVector(
   A=ScalarParameter() {defaultValue=-0.392, group=Ranked, parameterName=Arm_UV_OffsetX}
   B=ScalarParameter() {defaultValue=0.148, group=Ranked, parameterName=Arm_UV_OffsetY}
  )
  n61 = ScalarParameter() {defaultValue=0.35, group=Ranked, parameterName=UV_arm_Scale}
  n62 = AppendVector(
   A=ScalarParameter() {defaultValue=0.24, group=Ranked, parameterName=Arm_UV_ScaleX}
   B=ScalarParameter() {defaultValue=0.25, parameterName=Arm_UV_ScaleY}
  )
  NamedRerouteDeclaration(None=MaterialFunctionCall(UVs=Add(A=@n59, B=@n60), Texture Scale=Multiply(A=@n61, B=@n62)) {materialFunction=&ScaleUVsByCenter}.UVs) {name=UVArm, nodeColor=<rgba 0.4297 1 0 1>}
  NamedRerouteDeclaration(None=ComponentMask(None=@n16.M) {g=true}) {name=IN_AO, nodeColor=<rgba 0.6406 0 1 1>}
  NamedRerouteDeclaration(None=Power(Base=@n23.B)) {name=SkinDark, nodeColor=<rgba 0 1 0.5547 1>}
  NamedRerouteDeclaration(None=@n35.R) {name=MetalsMask, nodeColor=<rgba 0 0.3438 1 1>}
  n63 = Multiply(A=Desaturation(None=@n56, Fraction=Constant() {r=1}), B=@n57)
  NamedRerouteDeclaration(None=Max(A=NamedRerouteDeclaration(None=@n63) {name=EmissiveMask, nodeColor=<rgba 1 0.5391 0 1>}, B=@n58)) {name=EmissiveMask_eye, nodeColor=<rgba 1 0.5391 0 1>}
  NamedRerouteDeclaration(None=@n35.G) {name=SkullMask, nodeColor=<rgba 0.6641 1 0 1>}
  NamedRerouteDeclaration(None=MaterialFunctionCall() {materialFunction=/BRCosmetics/Characters/Player/Male/Medium/Bodies/M_MED_SpecterDunk/Materials/MF_ProjectNoise}.Result) {name=ColorNoise, nodeColor=<rgba 1 0.6562 0 1>}
 ]
}
